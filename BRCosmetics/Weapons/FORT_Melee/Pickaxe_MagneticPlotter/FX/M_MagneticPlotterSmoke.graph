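# Translucent unlit smoke material for Niagara sprites, located in the Magnetic Plotter pickaxe's FX.
# A panning, tiled swirl noise texture offsets the sprite texture UVs, with distortion strength driven per particle by a dynamic parameter.
# The sprite's green channel is multiplied by particle color for emissive, run through an emissive-adjustment function, and by particle alpha for opacity.
# Opacity is softened with depth fade and near-camera fade.
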Material /BRCosmetics/Weapons/FORT_Melee/Pickaxe_MagneticPlotter/FX/M_MagneticPlotterSmoke
graph {
 properties {bUsedWithNiagaraSprites=true, blendMode=BLEND_Translucent, pixelDepthOffsetMode=PDOM_Legacy, shadingModel=MSM_Unlit, usedShadingModels=Unlit}
 nodes [
  n1 = TextureCoordinate()
  n2 = ScalarParameter() {defaultValue=1, parameterName=DistortionTIling}
  n3 = TextureSample(UVs=Panner(Coordinate=Multiply(A=@n1, B=@n2)) {speedY=0.1}) {samplerType=SAMPLERTYPE_Normal, texture=/Game/Gadgets/Assets/VinderTech_GliderChute/Glider_Thrive_Female/FX/Texture/T_SwirlNoise_256}
  n4 = AppendVector(A=@n3.R, B=@n3.R)
  n5 = DynamicParameter() {defaultValue=<rgba 0.04 1 1 1>, paramNames [DistortionIntensity, Param2, Param3, Param4]}
  n6 = TextureSampleParameter2D(UVs=Add(A=TextureCoordinate(), B=Multiply(A=@n4, B=@n5.DistortionIntensity))) {
   parameterName=Tex01
   samplerType=SAMPLERTYPE_LinearColor
   texture=&/T_MagneticPlotterSprite
  }
  n7 = ParticleColor()
  MaterialFunctionCall(Emissive=Multiply(A=@n6.G, B=@n7.RGB)) {materialFunction=&MF_FX_EmissiveAdjustments}
  Multiply(
   A=DepthFade(Opacity=Multiply(A=@n6.G, B=@n7.A), FadeDistance=ScalarParameter() {defaultValue=10, group="1) Depth Fade", parameterName=Depth Fade})
   B=MaterialFunctionCall(Distance=ScalarParameter() {defaultValue=10, parameterName=Near Camera Fade}) {materialFunction=&Near_Camera_Fade}.Result
  )
 ]
}
